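# Niagara animation trail for the Kevin Couture male pickaxe, drawing a short-lived purple ribbon that follows the weapon.
# A single FN_AnimTrail_V4 emitter spawns via FN_SpawnAnimTrail_v2, then initializes and updates ribbon particles with shared AnimTrail bone-index and trail modules.
# Curves fade alpha and taper ribbon width over particle life.
# The User.FinalStage bool switches RGB brightness between dimmed and boosted.
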NiagaraSystem /BRCosmetics/Weapons/FORT_Melee/Pickaxe_Kevin_Couture_Male/FX/NS_Pickaxe_KevinCouture_AnimTrail
dependencies {usedDataInterfaces [&NiagaraDataInterfaceCurve], usedDynamicInputs [], usedModules [&SystemState]}
emitters {FN_AnimTrail_V4 {
  emitterUpdateScript [
   &EmitterState()
   &FN_SpawnAnimTrail_v2()
  ]
  particleSpawnScript [
   /Game/Effects/Niagara/Modules/AnimTrail/FN_InitializeAnimTrail/FN_InitializeAnimTrailBoneIndecies(First Bone or Socket Index=1)
   /Game/Effects/Niagara/Modules/AnimTrail/FN_InitializeAnimTrail/FN_InitializeAnimTrail(AnimTrailWidth=2, Color=<rgba 0.8971 0 1.5 0.85>, Lifetime=0.2)
  ]
  particleUpdateScript [
   &SolveForcesAndVelocity(Clamp Velocity=false, Limit Acceleration=false)
   /Game/Effects/Niagara/Modules/AnimTrail/FN_UpdateAnimTrail/FN_UpdateAnimTrail()
   &ScaleColor(
    Scale Alpha=&FloatFromCurve(FloatCurve=DataInterface() {propertyValues {Curve {keys [
         {arriveTangent=-17.1094, interpMode=RCIM_Cubic, leaveTangent=-17.1094, tangentMode=RCTM_User, time=0.005, value=2}
         {arriveTangent=-0.4835, interpMode=RCIM_Cubic, leaveTangent=-0.4835, tangentMode=RCTM_User, time=0.4, value=0.1515}
         {arriveTangent=-0.0191, interpMode=RCIM_Cubic, leaveTangent=-0.0191, tangentMode=RCTM_User, time=1, value=0}
        ]}}})
    Scale RGB=&VectorFromFloat(Value=&MakeCustomFloatFromBool(Bool=$User.FinalStage, True Float=2, False Float=0.6))
    ScaleA=true
    ScaleRGB=true
   )
   &ScaleRibbonWidth(Ribbon Width Scale=&FloatFromCurve(FloatCurve=DataInterface() {propertyValues {Curve {keys [
         {arriveTangent=-0.3, interpMode=RCIM_Cubic, time=0.1, value=1}
         {arriveTangent=-0.0294, interpMode=RCIM_Cubic, leaveTangent=-0.0294, tangentMode=RCTM_User, time=0.25, value=0.9932}
         {arriveTangent=-1.5611, interpMode=RCIM_Cubic, leaveTangent=-1.5611, tangentMode=RCTM_User, time=1, value=0}
        ]}}}))
  ]
  rendererClasses [&NiagaraRibbonRendererProperties]
  renderers [{rendererClass=&NiagaraRibbonRendererProperties}]
 }}
summary {
 systemName=NS_Pickaxe_KevinCouture_AnimTrail
 emitters [{emitterName=FN_AnimTrail_V4, rendererClasses [&NiagaraRibbonRendererProperties]}]
 userVariables [
  User.FinalStage {type=NiagaraBool}
  User.RibbonIDOffset {type=NiagaraInt32}
  User.SkeletalMesh {
   type=NiagaraDataInterfaceSkeletalMesh
   defaultValue {
    struct=NiagaraExt_VariableValue_DataInterface
    value {dataInterface=&self:NiagaraDataInterfaceSkeletalMesh_0}
   }
  }
 ]
}
