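# Material function for the Kevin Couture pickaxe that adds a wobbling World Position Offset, apparently to animated cubes on the weapon.
# Time times Wobble Speed drives sine and cosine waves with a different period per axis, forming an offset vector.
# A re-centered Cube ID input sets each cube's signed amplitude, further scaled by two multiplier parameters.
# The offset adds to the incoming Material Attributes' WPO, and an Alpha input blends between original and wobbled attributes.
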MaterialFunction /BRCosmetics/Weapons/FORT_Melee/Pickaxe_Kevin_Couture_Male/Materials/MF_WobblyWPOCubes
graph {nodes [
  n1 = Multiply(A=Time(), B=ScalarParameter() {defaultValue=0.5, parameterName=Wobble Speed})
  n2 = Sine(None=@n1) {period=9}
  Abs(None=@n2)
  ConstantBiasScale(None=@n2) {bias=0.5, scale=1}
  n3 = FunctionInput() {inputType=FunctionInput_MaterialAttributes}
  n4 = GetMaterialAttributes(None=@n3) {attributeGetTypes [#1]}
  n5 = ConstantBiasScale(None=FunctionInput() {bUsePreviewValueAsDefault=true, inputName=Cube ID, inputType=FunctionInput_Scalar}) {bias=-0.5, scale=8}
  n6 = MaterialFunctionCall(X=Multiply(A=Cosine(None=@n1) {period=7}, B=@n5), Y=Multiply(A=@n2, B=@n5), Z=Multiply(A=Sine(None=@n1) {period=11}, B=@n5)) {materialFunction=&MakeFloat3}
  n7 = ScalarParameter() {defaultValue=1.25, parameterName=WPOMultiply}
  n8 = Add(
   A=@n4.WorldPositionOffset
   B=Multiply(A=Multiply(A=@n6.Result, B=@n7), B=ScalarParameter() {defaultValue=1, parameterName=WPO_Mult_Override})
  )
  FunctionOutput(None=BlendMaterialAttributes(
    A=@n3
    B=SetMaterialAttributes!(MaterialAttributes=@n4.MaterialAttributes, WorldPositionOffset=@n8) {attributeSetTypes [#1]}
    Alpha=FunctionInput() {bUsePreviewValueAsDefault=true, inputName=Alpha, inputType=FunctionInput_Scalar}
   ))
 ]}
