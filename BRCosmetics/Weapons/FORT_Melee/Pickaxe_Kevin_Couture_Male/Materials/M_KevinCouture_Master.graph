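# Master material for the Kevin Couture male pickaxe, built on material attributes with the shared character PreFX and PostFX functions.
# PreFX output feeds a PanningEmissive layer, which a static switch, on by default, can bypass before PostFX.
# A separate branch passes the emissive result into MF_WobblyWPOCubes for wobbly cube vertex offset.
# One FX texture, through channel-mask parameters, supplies the emissive, UV-offset, cube ID and WPO masks.
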Material /BRCosmetics/Weapons/FORT_Melee/Pickaxe_Kevin_Couture_Male/Materials/M_KevinCouture_Master
graph {
 properties {
  bUseMaterialAttributes=true
  bUsedWithMeshDeformer=true
  bUsedWithNiagaraMeshParticles=true
  bUsedWithSkeletalMesh=true
  pixelDepthOffsetMode=PDOM_Legacy
  usedShadingModels=Default Lit
 }
 nodes [
  n1 = MaterialFunctionCall() {materialFunction=&BaseFN_CharacterFunction_PreFX}
  n2 = TextureSampleParameter2D() {parameterName=FX, texture=/Engine/EngineResources/Black_Low.Black}
  n3 = MaterialFunctionCall!(
   Alpha=ChannelMaskParameter(None=@n2.RGBA) {parameterName=Panning Emissive Mask}.R
   isVehicle
   Gmap=@n1.Gmap
   UVOffset=ChannelMaskParameter(None=@n2.RGBA) {parameterName=Panning Emissive Offset Mask}.R
   Local Normal
   Local Position
   EmissiveColorOverride
   Material=@n1.Gmap
  ) {materialFunction=/Game/Weapons/WeaponSkins/MaterialFunction/PanningEmissive}
  MaterialFunctionCall(
   AO=@n1.AO
   Distance Blend=@n1.AO
   Fresnel=@n1.AO
   Gmap=@n1.AO
   Material=StaticSwitchParameter(True=@n3.Result, False=@n1.Result) {defaultValue=true, parameterName=usePanningEmissive}
  ) {materialFunction=&BaseFN_CharacterFunction_PostFX}
  MaterialFunctionCall!(
   Cube ID=ChannelMaskParameter(None=@n2.RGBA) {parameterName=Cube FX Mask}.R
   Alpha=ChannelMaskParameter(None=@n2.RGBA) {parameterName=WPO mask}.R
   In=@n3.Result
  ) {materialFunction=&/MF_WobblyWPOCubes}
 ]
}
