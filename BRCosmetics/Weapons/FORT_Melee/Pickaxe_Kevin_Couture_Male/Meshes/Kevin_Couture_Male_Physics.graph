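# Physics asset for the Kevin Couture Male pickaxe cosmetic, giving collision bodies to its DYN-prefixed upper and lower cube bones.
# Each bone has one capsule, mostly zero-length and thus effectively spherical.
# Six upper cubes are constrained to the upper master body and one lower cube to the lower master, with no constraint linking the two masters.
# All constraints lock swing and twist, so their angular limits are inactive.
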PhysicsAsset /BRCosmetics/Weapons/FORT_Melee/Pickaxe_Kevin_Couture_Male/Meshes/Kevin_Couture_Male_Physics
bodies {
 DYN_LowerCube2_collision [{shapeType=Capsule, length=2.4193, radius=4.7533, shapeName=DYN_LowerCube2_collision_capsule, center=<-0.6797 0 0>}]
 DYN_LowerCubeMSTR [{shapeType=Capsule, length=0.2876, radius=8.4881, shapeName=DYN_LowerCubeMSTR_capsule}]
 DYN_UpperCube1_collision [{shapeType=Capsule, length=0, radius=4.75, shapeName=DYN_UpperCube1_collision_capsule}]
 DYN_UpperCube2_collision [{shapeType=Capsule, length=0, radius=4.75, shapeName=DYN_UpperCube2_collision_capsule, center=<0 -1.2902 0>}]
 DYN_UpperCube3_collision [{shapeType=Capsule, length=0, radius=6.25, shapeName=DYN_UpperCube3_collision_capsule, center=<-1.0617 0 0>}]
 DYN_UpperCube4_collision [{shapeType=Capsule, length=0, radius=5.25, shapeName=DYN_UpperCube4_collision_capsule, center=<0 -1.6166 -0.3628>}]
 DYN_UpperCube5_collision [{shapeType=Capsule, length=0, radius=8.25, shapeName=DYN_UpperCube5_collision_capsule, center=<-3.6006 0 0>}]
 DYN_UpperCube6_collision [{shapeType=Capsule, length=0, radius=6.75, shapeName=DYN_UpperCube6_collision_capsule}]
 DYN_UpperCubeMSTR [{shapeType=Capsule, length=0, radius=9.0856, shapeName=DYN_UpperCubeMSTR_capsule}]
}
constraints [
 {
  bone1Name=DYN_UpperCube2_collision
  bone2Name=DYN_UpperCubeMSTR
  swing1LimitDegrees=45
  swing1Motion=Locked
  swing2LimitDegrees=45
  swing2Motion=Locked
  twistLimitDegrees=45
  twistMotion=Locked
 }
 {
  bone1Name=DYN_UpperCube3_collision
  bone2Name=DYN_UpperCubeMSTR
  swing1LimitDegrees=45
  swing1Motion=Locked
  swing2LimitDegrees=45
  swing2Motion=Locked
  twistLimitDegrees=45
  twistMotion=Locked
 }
 {
  bone1Name=DYN_UpperCube4_collision
  bone2Name=DYN_UpperCubeMSTR
  swing1LimitDegrees=45
  swing1Motion=Locked
  swing2LimitDegrees=45
  swing2Motion=Locked
  twistLimitDegrees=45
  twistMotion=Locked
 }
 {
  bone1Name=DYN_UpperCube5_collision
  bone2Name=DYN_UpperCubeMSTR
  swing1LimitDegrees=45
  swing1Motion=Locked
  swing2LimitDegrees=45
  swing2Motion=Locked
  twistLimitDegrees=45
  twistMotion=Locked
 }
 {
  bone1Name=DYN_UpperCube6_collision
  bone2Name=DYN_UpperCubeMSTR
  swing1LimitDegrees=45
  swing1Motion=Locked
  swing2LimitDegrees=45
  swing2Motion=Locked
  twistLimitDegrees=45
  twistMotion=Locked
 }
 {
  bone1Name=DYN_LowerCube2_collision
  bone2Name=DYN_LowerCubeMSTR
  swing1LimitDegrees=45
  swing1Motion=Locked
  swing2LimitDegrees=45
  swing2Motion=Locked
  twistLimitDegrees=45
  twistMotion=Locked
 }
 {
  bone1Name=DYN_UpperCube1_collision
  bone2Name=DYN_UpperCubeMSTR
  swing1LimitDegrees=45
  swing1Motion=Locked
  swing2LimitDegrees=45
  swing2Motion=Locked
  twistLimitDegrees=45
  twistMotion=Locked
 }
]
