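# Animation Blueprint for the Kevin Couture Male pickaxe, giving its cube bones a floating, springy secondary motion.
# Additive bone transforms apply float and rotator variables, alongside time-based sine math in the event graph.
# Gravity-free AnimDynamics chains spring each cube, world-space for front end and handle-relative otherwise.
# Bool blends select chains by front-end, ground and melee state; the cached 'OnGroundCache' chain adds a RigidBody pass.
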AnimBlueprint /BRCosmetics/Weapons/FORT_Melee/Pickaxe_Kevin_Couture_Male/Meshes/Kevin_Couture_Male_AnimBP
target_skeleton=&/Kevin_Couture_Male_Skeleton
graphs [
 AnimGraph {nodes [
   Local To Component()
   n1 = "Transform (Modify) Bone - Bone: DYN_UpperCube4"(
    ComponentPose=Local To Component(LocalPose=Use cached pose 'OnGroundCache'())
    Alpha=1
    Translation="-6,-2,-2"
    Rotation="0,0,0"
    Scale="1,1,1"
   ) {
    translationMode=BMM_Additive
    translationSpace=BCS_BoneSpace
    alphaBoolBlend {blendOption=Linear}
    boneToModify {boneName=DYN_UpperCube4}
    translation=<-6 -2 -2>
   }
   n2 = "Transform (Modify) Bone - Bone: DYN_UpperCube3"(ComponentPose=@n1, Alpha=1, Translation="-3,-3,1", Rotation="0,0,0", Scale="1,1,1") {
    translationMode=BMM_Additive
    translationSpace=BCS_BoneSpace
    alphaBoolBlend {blendOption=Linear}
    boneToModify {boneName=DYN_UpperCube3}
    translation=<-3 -3 1>
   }
   n3 = Get Cube_Float_A() -> (Cube_Float_A:Vector)
   n4 = Get Floating_Rotator2() -> (Floating_Rotator2:Rotator)
   n5 = "Transform (Modify) Bone - Bone: DYN_UpperCube1"(ComponentPose=@n2, Alpha=1, Translation=@n3, Rotation=@n4, Scale="1,1,1") {rotationMode=BMM_Additive, translationMode=BMM_Additive, alphaBoolBlend {blendOption=Linear}, boneToModify {boneName=DYN_UpperCube1}}
   n6 = "Transform (Modify) Bone - Bone: DYN_UpperCube2"(ComponentPose=@n5, Alpha=1, Translation=Get Cube_Float_B() -> (Cube_Float_B:Vector), Rotation=@n4, Scale="1,1,1") {rotationMode=BMM_Additive, translationMode=BMM_Additive, alphaBoolBlend {blendOption=Linear}, boneToModify {boneName=DYN_UpperCube2}}
   n7 = "Transform (Modify) Bone - Bone: DYN_UpperCube3"(ComponentPose=@n6, Alpha=1, Translation=@n3, Rotation=@n4, Scale="1,1,1") {rotationMode=BMM_Additive, translationMode=BMM_Additive, alphaBoolBlend {blendOption=Linear}, boneToModify {boneName=DYN_UpperCube3}}
   n8 = Get Cube_Float_C() -> (Cube_Float_C:Vector)
   n9 = Get Floating_Rotator1() -> (Floating_Rotator1:Rotator)
   n10 = "Transform (Modify) Bone - Bone: DYN_UpperCube4"(ComponentPose=@n7, Alpha=1, Translation=@n8, Rotation=@n9, Scale="1,1,1") {rotationMode=BMM_Additive, translationMode=BMM_Additive, alphaBoolBlend {blendOption=Linear}, boneToModify {boneName=DYN_UpperCube4}}
   n11 = Get Cube_Float_D() -> (Cube_Float_D:Vector)
   n12 = "Transform (Modify) Bone - Bone: DYN_UpperCube5"(ComponentPose=@n10, Alpha=1, Translation=@n11, Rotation=@n9, Scale="1,1,1") {rotationMode=BMM_Additive, translationMode=BMM_Additive, alphaBoolBlend {blendOption=Linear}, boneToModify {boneName=DYN_UpperCube5}}
   n13 = "Transform (Modify) Bone - Bone: DYN_UpperCube6"(ComponentPose=@n12, Alpha=1, Translation=@n8, Rotation=@n9, Scale="1,1,1") {rotationMode=BMM_Additive, translationMode=BMM_Additive, alphaBoolBlend {blendOption=Linear}, boneToModify {boneName=DYN_UpperCube6}}
   n14 = "Transform (Modify) Bone - Bone: DYN_LowerCube2"(ComponentPose=@n13, Alpha=1, Translation=@n11, Rotation=@n9, Scale="1,1,1") {rotationMode=BMM_Additive, translationMode=BMM_Additive, alphaBoolBlend {blendOption=Linear}, boneToModify {boneName=DYN_LowerCube2}}
   n15 = Anim Dynamics - Bone: DYN_UpperCube1(ComponentPose=@n14, Alpha=0.9, ExternalForce="0,0,0") {
    angularDampingOverride=0.9
    angularSpringConstant=175
    bAngularSpring=true
    bGravityOverrideInSimSpace=true
    bLinearSpring=true
    bOverrideAngularDamping=true
    bOverrideLinearDamping=true
    gravityScale=0
    linearDampingOverride=0.995
    linearSpringConstant=155
    simulationSpace=BoneRelative
    alphaBoolBlend {blendOption=Linear}
    boundBone {boneName=DYN_UpperCube1}
    componentAppliedLinearAccClamp=<6000 6000 6000>
    componentLinearVelScale=<0.9 0.9 0.9>
    physicsBodyDefinitions [{
      collisionType=InnerSphere
      sphereCollisionRadius=0
      boundBone {boneName=DYN_UpperCube1}
      boxExtents=<50 50 50>
      constraintSetup {angularLimitsMax=<0 0 0>, angularLimitsMin=<0 0 0>, angularTarget=<0 0 0>, linearAxesMax=<7 8 8>, linearAxesMin=<-8 -2 -5>}
      localJointOffset=<-3 -3 -3>
     }]
    relativeSpaceBone {boneName=Melee_TwoHanded_Handle}
   }
   n16 = Anim Dynamics - Bone: DYN_UpperCube2(ComponentPose=@n15, Alpha=0.9, ExternalForce="0,0,0") {
    angularDampingOverride=0.9
    angularSpringConstant=195
    bAngularSpring=true
    bGravityOverrideInSimSpace=true
    bLinearSpring=true
    bOverrideAngularDamping=true
    bOverrideLinearDamping=true
    gravityScale=0
    linearDampingOverride=0.995
    linearSpringConstant=155
    simulationSpace=BoneRelative
    alphaBoolBlend {blendOption=Linear}
    boundBone {boneName=DYN_UpperCube2}
    componentAppliedLinearAccClamp=<6000 6000 6000>
    componentLinearVelScale=<0.9 0.9 0.9>
    physicsBodyDefinitions [{
      sphereCollisionRadius=0
      boundBone {boneName=DYN_UpperCube2}
      boxExtents=<50 50 50>
      constraintSetup {angularLimitsMax=<0 0 0>, angularLimitsMin=<0 0 0>, angularTarget=<0 0 0>, linearAxesMax=<8 3 8>, linearAxesMin=<-5 -8 -4>}
      localJointOffset=<-3 -3 -3>
     }]
    relativeSpaceBone {boneName=Melee_TwoHanded_Handle}
   }
   n17 = Anim Dynamics - Bone: DYN_UpperCube3(ComponentPose=@n16, Alpha=0.9, ExternalForce="0,0,0") {
    angularDampingOverride=0.9
    angularSpringConstant=195
    bAngularSpring=true
    bGravityOverrideInSimSpace=true
    bLinearSpring=true
    bOverrideAngularDamping=true
    bOverrideLinearDamping=true
    gravityScale=0
    linearDampingOverride=0.995
    linearSpringConstant=155
    simulationSpace=BoneRelative
    alphaBoolBlend {blendOption=Linear}
    boundBone {boneName=DYN_UpperCube3}
    componentAppliedLinearAccClamp=<6000 6000 6000>
    componentLinearVelScale=<0.9 0.9 0.9>
    physicsBodyDefinitions [{
      sphereCollisionRadius=0
      boundBone {boneName=DYN_UpperCube3}
      boxExtents=<50 50 50>
      constraintSetup {angularLimitsMax=<0 0 0>, angularLimitsMin=<0 0 0>, angularTarget=<0 0 0>, linearAxesMax=<8 8 6>, linearAxesMin=<-8 -2 -3>}
      localJointOffset=<-3 -3 -3>
     }]
    relativeSpaceBone {boneName=Melee_TwoHanded_Handle}
   }
   n18 = Anim Dynamics - Bone: DYN_UpperCube4(ComponentPose=@n17, Alpha=0.9, ExternalForce="0,0,0") {
    angularDampingOverride=0.9
    angularSpringConstant=175
    bAngularSpring=true
    bGravityOverrideInSimSpace=true
    bLinearSpring=true
    bOverrideAngularDamping=true
    bOverrideLinearDamping=true
    gravityScale=0
    linearDampingOverride=0.995
    linearSpringConstant=155
    simulationSpace=BoneRelative
    alphaBoolBlend {blendOption=Linear}
    boundBone {boneName=DYN_UpperCube4}
    componentAppliedLinearAccClamp=<6000 6000 6000>
    componentLinearVelScale=<0.9 0.9 0.9>
    physicsBodyDefinitions [{
      sphereCollisionRadius=0
      boundBone {boneName=DYN_UpperCube4}
      boxExtents=<50 50 50>
      constraintSetup {angularLimitsMax=<0 0 0>, angularLimitsMin=<0 0 0>, angularTarget=<0 0 0>, linearAxesMax=<2 3 8>, linearAxesMin=<-8 -8 -2>}
      localJointOffset=<-3 -3 -3>
     }]
    relativeSpaceBone {boneName=Melee_TwoHanded_Handle}
   }
   n19 = Anim Dynamics - Bone: DYN_UpperCube5(ComponentPose=@n18, Alpha=0.9, ExternalForce="0,0,0") {
    angularDampingOverride=0.9
    angularSpringConstant=175
    bAngularSpring=true
    bGravityOverrideInSimSpace=true
    bLinearSpring=true
    bOverrideAngularDamping=true
    bOverrideLinearDamping=true
    gravityScale=0
    linearDampingOverride=0.995
    linearSpringConstant=155
    simulationSpace=BoneRelative
    alphaBoolBlend {blendOption=Linear}
    boundBone {boneName=DYN_UpperCube5}
    componentAppliedLinearAccClamp=<6000 6000 6000>
    componentLinearVelScale=<0.9 0.9 0.9>
    physicsBodyDefinitions [{
      sphereCollisionRadius=0
      boundBone {boneName=DYN_UpperCube5}
      boxExtents=<50 50 50>
      constraintSetup {angularLimitsMax=<0 0 0>, angularLimitsMin=<0 0 0>, angularTarget=<0 0 0>, linearAxesMax=<4 8 8>, linearAxesMin=<-8 -2 -8>}
      localJointOffset=<-3 -3 -3>
     }]
    relativeSpaceBone {boneName=Melee_TwoHanded_Handle}
   }
   n20 = Anim Dynamics - Bone: DYN_UpperCube6(ComponentPose=@n19, Alpha=0.5, ExternalForce="0,0,0") {
    angularDampingOverride=0.9
    angularSpringConstant=150
    bAngularSpring=true
    bGravityOverrideInSimSpace=true
    bLinearSpring=true
    bOverrideAngularDamping=true
    bOverrideLinearDamping=true
    gravityScale=0
    linearDampingOverride=0.995
    linearSpringConstant=50
    simulationSpace=BoneRelative
    alphaBoolBlend {blendOption=Linear}
    boundBone {boneName=DYN_UpperCube6}
    componentAppliedLinearAccClamp=<6000 6000 6000>
    componentLinearAccScale=<0.1 0.1 0.1>
    componentLinearVelScale=<0.9 0.9 0.9>
    physicsBodyDefinitions [{
      sphereCollisionRadius=0
      boundBone {boneName=DYN_UpperCube6}
      boxExtents=<50 50 50>
      constraintSetup {angularLimitsMax=<5 5 5>, angularLimitsMin=<-5 -5 -5>, angularTarget=<0 0 0>, linearAxesMax=<2 8 2>, linearAxesMin=<-8 -8 -8>}
      localJointOffset=<-3 0 0>
     }]
    relativeSpaceBone {boneName=Melee_TwoHanded_Handle}
   }
   n21 = Anim Dynamics - Bone: DYN_LowerCube2(ComponentPose=@n20, Alpha=1, ExternalForce="0,0,0") {
    angularDampingOverride=0.9
    angularSpringConstant=175
    bAngularSpring=true
    bGravityOverrideInSimSpace=true
    bLinearSpring=true
    bOverrideAngularDamping=true
    bOverrideLinearDamping=true
    gravityScale=0
    linearDampingOverride=0.995
    linearSpringConstant=155
    simulationSpace=BoneRelative
    alphaBoolBlend {blendOption=Linear}
    boundBone {boneName=DYN_LowerCube2}
    componentAppliedLinearAccClamp=<6000 6000 6000>
    componentLinearVelScale=<0.9 0.9 0.9>
    physicsBodyDefinitions [{
      sphereCollisionRadius=0
      boundBone {boneName=DYN_LowerCube2}
      boxExtents=<50 50 50>
      constraintSetup {angularLimitsMax=<0 0 0>, angularLimitsMin=<0 0 0>, angularTarget=<0 0 0>, linearAxesMax=<2 8 8>, linearAxesMin=<-8 -3 -8>}
      localJointOffset=<0 -3 0>
     }]
    relativeSpaceBone {boneName=Melee_TwoHanded_Handle}
   }
   n22 = Component To Local(ComponentPose=@n21)
   n23 = "Transform (Modify) Bone - Bone: DYN_UpperCube4"(ComponentPose=Local To Component(LocalPose=Input Pose() {name=InPose_0}), Alpha=1, Translation="-2,-1,-1", Rotation="0,0,0", Scale="1,1,1") {
    translationMode=BMM_Additive
    translationSpace=BCS_BoneSpace
    alphaBoolBlend {blendOption=Linear}
    boneToModify {boneName=DYN_UpperCube4}
    translation=<-2 -1 -1>
   }
   n24 = "Transform (Modify) Bone - Bone: DYN_UpperCube3"(ComponentPose=@n23, Alpha=1, Translation="-3,-3,1", Rotation="0,0,0", Scale="1,1,1") {
    translationMode=BMM_Additive
    translationSpace=BCS_BoneSpace
    alphaBoolBlend {blendOption=Linear}
    boneToModify {boneName=DYN_UpperCube3}
    translation=<-3 -3 1>
   }
   n25 = Get Cube_Float_A()
   n26 = Get Floating_Rotator2()
   n27 = "Transform (Modify) Bone - Bone: DYN_UpperCube1"(ComponentPose=@n24, Alpha=1, Translation=@n25, Rotation=@n26, Scale="1,1,1") {rotationMode=BMM_Additive, translationMode=BMM_Additive, alphaBoolBlend {blendOption=Linear}, boneToModify {boneName=DYN_UpperCube1}}
   n28 = "Transform (Modify) Bone - Bone: DYN_UpperCube2"(ComponentPose=@n27, Alpha=0.5, Translation=Get Cube_Float_B(), Rotation=@n26, Scale="1,1,1") {rotationMode=BMM_Additive, translationMode=BMM_Additive, alphaBoolBlend {blendOption=Linear}, boneToModify {boneName=DYN_UpperCube2}}
   n29 = Get Floating_Rotator1()
   n30 = "Transform (Modify) Bone - Bone: DYN_UpperCube3"(ComponentPose=@n28, Alpha=0.5, Translation=@n25, Rotation=@n29, Scale="1,1,1") {rotationMode=BMM_Additive, translationMode=BMM_Additive, alphaBoolBlend {blendOption=Linear}, boneToModify {boneName=DYN_UpperCube3}}
   n31 = Get Cube_Float_C()
   n32 = "Transform (Modify) Bone - Bone: DYN_UpperCube4"(ComponentPose=@n30, Alpha=1, Translation=@n31, Rotation=@n29, Scale="1,1,1") {rotationMode=BMM_Additive, translationMode=BMM_Additive, alphaBoolBlend {blendOption=Linear}, boneToModify {boneName=DYN_UpperCube4}}
   n33 = Get Cube_Float_D()
   n34 = "Transform (Modify) Bone - Bone: DYN_UpperCube5"(ComponentPose=@n32, Alpha=1, Translation=@n33, Rotation=@n29, Scale="1,1,1") {rotationMode=BMM_Additive, translationMode=BMM_Additive, alphaBoolBlend {blendOption=Linear}, boneToModify {boneName=DYN_UpperCube5}}
   n35 = "Transform (Modify) Bone - Bone: DYN_UpperCube6"(ComponentPose=@n34, Alpha=1, Translation=@n31, Rotation=@n26, Scale="1,1,1") {rotationMode=BMM_Additive, translationMode=BMM_Additive, alphaBoolBlend {blendOption=Linear}, boneToModify {boneName=DYN_UpperCube6}}
   n36 = "Transform (Modify) Bone - Bone: DYN_LowerCube2"(ComponentPose=@n35, Alpha=1, Translation=@n33, Rotation=@n29, Scale="1,1,1") {rotationMode=BMM_Additive, translationMode=BMM_Additive, alphaBoolBlend {blendOption=Linear}, boneToModify {boneName=DYN_LowerCube2}}
   n37 = Anim Dynamics - Bone: DYN_UpperCube1(ComponentPose=@n36, Alpha=0.9, ExternalForce="0,0,0") {
    angularDampingOverride=0.9
    angularSpringConstant=150
    bAngularSpring=true
    bGravityOverrideInSimSpace=true
    bLinearSpring=true
    bOverrideAngularDamping=true
    bOverrideLinearDamping=true
    gravityScale=0
    linearDampingOverride=0.995
    linearSpringConstant=25
    simulationSpace=BoneRelative
    alphaBoolBlend {blendOption=Linear}
    boundBone {boneName=DYN_UpperCube1}
    componentAppliedLinearAccClamp=<6000 6000 6000>
    componentLinearAccScale=<0.13 0.13 0.13>
    componentLinearVelScale=<0.9 0.9 0.9>
    physicsBodyDefinitions [{
      sphereCollisionRadius=0
      boundBone {boneName=DYN_UpperCube1}
      boxExtents=<50 50 50>
      constraintSetup {angularLimitsMax=<5 5 5>, angularLimitsMin=<-5 -5 -5>, angularTarget=<0 0 0>, linearAxesMax=<7 8 8>, linearAxesMin=<-8 -2 -5>}
      localJointOffset=<0 -3 0>
     }]
    relativeSpaceBone {boneName=Melee_TwoHanded_Handle}
   }
   n38 = Anim Dynamics - Bone: DYN_UpperCube2(ComponentPose=@n37, Alpha=0.9, ExternalForce="0,0,0") {
    angularDampingOverride=0.9
    angularSpringConstant=150
    bAngularSpring=true
    bGravityOverrideInSimSpace=true
    bLinearSpring=true
    bOverrideAngularDamping=true
    bOverrideLinearDamping=true
    gravityScale=0
    linearDampingOverride=0.995
    linearSpringConstant=25
    simulationSpace=BoneRelative
    alphaBoolBlend {blendOption=Linear}
    boundBone {boneName=DYN_UpperCube2}
    componentAppliedLinearAccClamp=<6000 6000 6000>
    componentLinearAccScale=<0.1 0.1 0.1>
    componentLinearVelScale=<0.9 0.9 0.9>
    physicsBodyDefinitions [{
      sphereCollisionRadius=0
      boundBone {boneName=DYN_UpperCube2}
      boxExtents=<50 50 50>
      constraintSetup {angularLimitsMax=<5 5 5>, angularLimitsMin=<-5 -5 -5>, angularTarget=<0 0 0>, linearAxesMax=<8 3 8>, linearAxesMin=<-5 -8 -4>}
      localJointOffset=<0 -3 0>
     }]
    relativeSpaceBone {boneName=Melee_TwoHanded_Handle}
   }
   n39 = Anim Dynamics - Bone: DYN_UpperCube3(ComponentPose=@n38, Alpha=0.7, ExternalForce="0,0,0") {
    angularDampingOverride=0.9
    angularSpringConstant=150
    bAngularSpring=true
    bGravityOverrideInSimSpace=true
    bLinearSpring=true
    bOverrideAngularDamping=true
    bOverrideLinearDamping=true
    gravityScale=0
    linearDampingOverride=0.995
    linearSpringConstant=25
    simulationSpace=BoneRelative
    alphaBoolBlend {blendOption=Linear}
    boundBone {boneName=DYN_UpperCube3}
    componentAppliedLinearAccClamp=<6000 6000 6000>
    componentLinearAccScale=<0.1 0.1 0.1>
    componentLinearVelScale=<0.9 0.9 0.9>
    physicsBodyDefinitions [{
      sphereCollisionRadius=0
      boundBone {boneName=DYN_UpperCube3}
      boxExtents=<50 50 50>
      constraintSetup {angularLimitsMax=<5 5 5>, angularLimitsMin=<-5 -5 -5>, angularTarget=<0 0 0>, linearAxesMax=<8 8 6>, linearAxesMin=<-7 0 0>}
      localJointOffset=<0 -3 0>
     }]
    relativeSpaceBone {boneName=Melee_TwoHanded_Handle}
   }
   n40 = Anim Dynamics - Bone: DYN_UpperCube4(ComponentPose=@n39, Alpha=0.9, ExternalForce="0,0,0") {
    angularDampingOverride=0.9
    angularSpringConstant=150
    bAngularSpring=true
    bGravityOverrideInSimSpace=true
    bLinearSpring=true
    bOverrideAngularDamping=true
    bOverrideLinearDamping=true
    gravityScale=0
    linearDampingOverride=0.995
    linearSpringConstant=25
    simulationSpace=BoneRelative
    alphaBoolBlend {blendOption=Linear}
    boundBone {boneName=DYN_UpperCube4}
    componentAppliedLinearAccClamp=<6000 6000 6000>
    componentLinearAccScale=<0.1 0.1 0.1>
    componentLinearVelScale=<0.9 0.9 0.9>
    physicsBodyDefinitions [{
      sphereCollisionRadius=0
      boundBone {boneName=DYN_UpperCube4}
      boxExtents=<50 50 50>
      constraintSetup {angularLimitsMax=<5 5 5>, angularLimitsMin=<-5 -5 -5>, angularTarget=<0 0 0>, linearAxesMax=<2 3 8>, linearAxesMin=<-8 -8 -2>}
      localJointOffset=<-3 0 0>
     }]
    relativeSpaceBone {boneName=Melee_TwoHanded_Handle}
   }
   n41 = Anim Dynamics - Bone: DYN_UpperCube5(ComponentPose=@n40, Alpha=0.9, ExternalForce="0,0,0") {
    angularDampingOverride=0.9
    angularSpringConstant=150
    bAngularSpring=true
    bGravityOverrideInSimSpace=true
    bLinearSpring=true
    bOverrideAngularDamping=true
    bOverrideLinearDamping=true
    gravityScale=0
    linearDampingOverride=0.995
    linearSpringConstant=25
    simulationSpace=BoneRelative
    alphaBoolBlend {blendOption=Linear}
    boundBone {boneName=DYN_UpperCube5}
    componentAppliedLinearAccClamp=<6000 6000 6000>
    componentLinearAccScale=<0.1 0.1 0.1>
    componentLinearVelScale=<0.9 0.9 0.9>
    physicsBodyDefinitions [{
      sphereCollisionRadius=0
      boundBone {boneName=DYN_UpperCube5}
      boxExtents=<50 50 50>
      constraintSetup {angularLimitsMax=<5 5 5>, angularLimitsMin=<-5 -5 -5>, angularTarget=<0 0 0>, linearAxesMax=<4 8 8>, linearAxesMin=<-8 -2 -8>}
      localJointOffset=<0 -3 0>
     }]
    relativeSpaceBone {boneName=Melee_TwoHanded_Handle}
   }
   n42 = Anim Dynamics - Bone: DYN_UpperCube6(ComponentPose=@n41, Alpha=0.7, ExternalForce="0,0,0") {
    angularDampingOverride=0.9
    angularSpringConstant=150
    bAngularSpring=true
    bGravityOverrideInSimSpace=true
    bLinearSpring=true
    bOverrideAngularDamping=true
    bOverrideLinearDamping=true
    gravityScale=0
    linearDampingOverride=0.995
    linearSpringConstant=50
    simulationSpace=BoneRelative
    alphaBoolBlend {blendOption=Linear}
    boundBone {boneName=DYN_UpperCube6}
    componentAppliedLinearAccClamp=<6000 6000 6000>
    componentLinearAccScale=<0.1 0.1 0.1>
    componentLinearVelScale=<0.9 0.9 0.9>
    physicsBodyDefinitions [{
      sphereCollisionRadius=0
      boundBone {boneName=DYN_UpperCube6}
      boxExtents=<50 50 50>
      constraintSetup {angularLimitsMax=<5 5 5>, angularLimitsMin=<-5 -5 -5>, angularTarget=<0 0 0>, linearAxesMax=<2 8 2>, linearAxesMin=<-8 -8 -8>}
      localJointOffset=<-3 0 0>
     }]
    relativeSpaceBone {boneName=Melee_TwoHanded_Handle}
   }
   n43 = Anim Dynamics - Bone: DYN_LowerCube2(ComponentPose=@n42, Alpha=0.9, ExternalForce="0,0,0") {
    angularDampingOverride=0.9
    angularSpringConstant=150
    bAngularSpring=true
    bGravityOverrideInSimSpace=true
    bLinearSpring=true
    bOverrideAngularDamping=true
    bOverrideLinearDamping=true
    gravityScale=0
    linearDampingOverride=0.995
    linearSpringConstant=25
    simulationSpace=BoneRelative
    alphaBoolBlend {blendOption=Linear}
    boundBone {boneName=DYN_LowerCube2}
    componentAppliedLinearAccClamp=<6000 6000 6000>
    componentLinearAccScale=<0.1 0.1 0.1>
    componentLinearVelScale=<0.9 0.9 0.9>
    physicsBodyDefinitions [{
      sphereCollisionRadius=0
      boundBone {boneName=DYN_LowerCube2}
      boxExtents=<50 50 50>
      constraintSetup {angularLimitsMax=<0 0 0>, angularLimitsMin=<0 0 0>, angularTarget=<0 0 0>, linearAxesMax=<2 8 8>, linearAxesMin=<-8 -3 -8>}
      localJointOffset=<0 -3 0>
     }]
    relativeSpaceBone {boneName=Melee_TwoHanded_Handle}
   }
   n44 = Component To Local(ComponentPose=@n43)
   n45 = OR Boolean(A=Get bIsOnGround() -> (bIsOnGround:bool), B=Get bIsPlayingMeleeAnim())
   n46 = Anim Dynamics - Bone: DYN_UpperCube1(ComponentPose=Local To Component(LocalPose=Input Pose() {name=InPose_1}), Alpha=0.5, ExternalForce="0,0,0") {
    angularDampingOverride=0.9
    angularSpringConstant=150
    bAngularSpring=true
    bGravityOverrideInSimSpace=true
    bLinearSpring=true
    bOverrideAngularDamping=true
    bOverrideLinearDamping=true
    gravityScale=0
    linearDampingOverride=0.999
    linearSpringConstant=20
    simulationSpace=World
    alphaBoolBlend {blendOption=Linear}
    boundBone {boneName=DYN_UpperCube1}
    componentAppliedLinearAccClamp=<50 50 50>
    componentLinearVelScale=<0.9 0.9 0.9>
    physicsBodyDefinitions [{
      sphereCollisionRadius=0
      boundBone {boneName=DYN_UpperCube1}
      boxExtents=<50 50 50>
      constraintSetup {angularLimitsMax=<5 5 5>, angularLimitsMin=<-5 -5 -5>, angularTarget=<0 0 0>, linearAxesMax=<14 20 20>, linearAxesMin=<-20 -4 -10>}
      localJointOffset=<0 -3 0>
     }]
    relativeSpaceBone {boneName=DYN_UpperCubeMSTR}
   }
   n47 = Anim Dynamics - Bone: DYN_UpperCube2(ComponentPose=@n46, Alpha=0.5, ExternalForce="0,0,0") {
    angularDampingOverride=0.9
    angularSpringConstant=150
    bAngularSpring=true
    bGravityOverrideInSimSpace=true
    bLinearSpring=true
    bOverrideAngularDamping=true
    bOverrideLinearDamping=true
    gravityScale=0
    linearDampingOverride=0.999
    linearSpringConstant=20
    simulationSpace=World
    alphaBoolBlend {blendOption=Linear}
    boundBone {boneName=DYN_UpperCube2}
    componentAppliedLinearAccClamp=<50 50 50>
    componentLinearVelScale=<0.9 0.9 0.9>
    physicsBodyDefinitions [{
      sphereCollisionRadius=0
      boundBone {boneName=DYN_UpperCube2}
      boxExtents=<50 50 50>
      constraintSetup {angularLimitsMax=<5 5 5>, angularLimitsMin=<-5 -5 -5>, angularTarget=<0 0 0>, linearAxesMax=<20 6 20>, linearAxesMin=<-10 -20 -8>}
      localJointOffset=<0 -3 0>
     }]
    relativeSpaceBone {boneName=DYN_UpperCubeMSTR}
   }
   n48 = Anim Dynamics - Bone: DYN_UpperCube3(ComponentPose=@n47, Alpha=0.5, ExternalForce="0,0,0") {
    angularDampingOverride=0.9
    angularSpringConstant=150
    bAngularSpring=true
    bGravityOverrideInSimSpace=true
    bLinearSpring=true
    bOverrideAngularDamping=true
    bOverrideLinearDamping=true
    gravityScale=0
    linearDampingOverride=0.999
    linearSpringConstant=20
    simulationSpace=World
    alphaBoolBlend {blendOption=Linear}
    boundBone {boneName=DYN_UpperCube3}
    componentAppliedLinearAccClamp=<50 50 50>
    componentLinearVelScale=<0.9 0.9 0.9>
    physicsBodyDefinitions [{
      sphereCollisionRadius=0
      boundBone {boneName=DYN_UpperCube3}
      boxExtents=<50 50 50>
      constraintSetup {angularLimitsMax=<5 5 5>, angularLimitsMin=<-5 -5 -5>, angularTarget=<0 0 0>, linearAxesMax=<20 20 12>, linearAxesMin=<-14 0 0>}
      localJointOffset=<0 -3 0>
     }]
    relativeSpaceBone {boneName=DYN_UpperCubeMSTR}
   }
   n49 = Anim Dynamics - Bone: DYN_UpperCube4(ComponentPose=@n48, Alpha=0.5, ExternalForce="0,0,0") {
    angularDampingOverride=0.9
    angularSpringConstant=150
    bAngularSpring=true
    bGravityOverrideInSimSpace=true
    bLinearSpring=true
    bOverrideAngularDamping=true
    bOverrideLinearDamping=true
    gravityScale=0
    linearDampingOverride=0.999
    linearSpringConstant=20
    simulationSpace=World
    alphaBoolBlend {blendOption=Linear}
    boundBone {boneName=DYN_UpperCube4}
    componentAppliedLinearAccClamp=<50 50 50>
    componentLinearVelScale=<0.9 0.9 0.9>
    physicsBodyDefinitions [{
      sphereCollisionRadius=0
      boundBone {boneName=DYN_UpperCube4}
      boxExtents=<50 50 50>
      constraintSetup {angularLimitsMax=<5 5 5>, angularLimitsMin=<-5 -5 -5>, angularTarget=<0 0 0>, linearAxesMax=<4 6 20>, linearAxesMin=<-20 -20 -4>}
      localJointOffset=<-3 0 0>
     }]
    relativeSpaceBone {boneName=DYN_UpperCubeMSTR}
   }
   n50 = Anim Dynamics - Bone: DYN_UpperCube5(ComponentPose=@n49, Alpha=0.5, ExternalForce="0,0,0") {
    angularDampingOverride=0.9
    angularSpringConstant=150
    bAngularSpring=true
    bGravityOverrideInSimSpace=true
    bLinearSpring=true
    bOverrideAngularDamping=true
    bOverrideLinearDamping=true
    gravityScale=0
    linearDampingOverride=0.999
    linearSpringConstant=20
    simulationSpace=World
    alphaBoolBlend {blendOption=Linear}
    boundBone {boneName=DYN_UpperCube5}
    componentAppliedLinearAccClamp=<50 50 50>
    componentLinearVelScale=<0.9 0.9 0.9>
    physicsBodyDefinitions [{
      sphereCollisionRadius=0
      boundBone {boneName=DYN_UpperCube5}
      boxExtents=<50 50 50>
      constraintSetup {angularLimitsMax=<5 5 5>, angularLimitsMin=<-5 -5 -5>, angularTarget=<0 0 0>, linearAxesMax=<8 20 20>, linearAxesMin=<-20 -4 -20>}
      localJointOffset=<0 -3 0>
     }]
    relativeSpaceBone {boneName=DYN_UpperCubeMSTR}
   }
   n51 = Anim Dynamics - Bone: DYN_UpperCube6(ComponentPose=@n50, Alpha=0.5, ExternalForce="0,0,0") {
    angularDampingOverride=0.9
    angularSpringConstant=150
    bAngularSpring=true
    bGravityOverrideInSimSpace=true
    bLinearSpring=true
    bOverrideAngularDamping=true
    bOverrideLinearDamping=true
    gravityScale=0
    linearDampingOverride=0.999
    linearSpringConstant=20
    simulationSpace=World
    alphaBoolBlend {blendOption=Linear}
    boundBone {boneName=DYN_UpperCube6}
    componentAppliedLinearAccClamp=<50 50 50>
    componentLinearVelScale=<0.9 0.9 0.9>
    physicsBodyDefinitions [{
      sphereCollisionRadius=0
      boundBone {boneName=DYN_UpperCube6}
      boxExtents=<50 50 50>
      constraintSetup {angularLimitsMax=<5 5 5>, angularLimitsMin=<-5 -5 -5>, angularTarget=<0 0 0>, linearAxesMax=<4 16 4>, linearAxesMin=<-20 -20 -20>}
      localJointOffset=<-3 0 0>
     }]
    relativeSpaceBone {boneName=DYN_UpperCubeMSTR}
   }
   n52 = Anim Dynamics - Bone: DYN_LowerCube2(ComponentPose=@n51, Alpha=1, ExternalForce="0,0,0") {
    angularDampingOverride=0.9
    angularSpringConstant=150
    bAngularSpring=true
    bGravityOverrideInSimSpace=true
    bLinearSpring=true
    bOverrideAngularDamping=true
    bOverrideLinearDamping=true
    gravityScale=0
    linearDampingOverride=0.999
    linearSpringConstant=15
    simulationSpace=World
    alphaBoolBlend {blendOption=Linear}
    boundBone {boneName=DYN_LowerCube2}
    componentAppliedLinearAccClamp=<1000 1000 1000>
    componentLinearVelScale=<2 2 2>
    physicsBodyDefinitions [{
      sphereCollisionRadius=0
      boundBone {boneName=DYN_LowerCube2}
      boxExtents=<50 50 50>
      constraintSetup {angularLimitsMax=<0 0 0>, angularLimitsMin=<0 0 0>, angularTarget=<0 0 0>, linearAxesMax=<2 8 8>, linearAxesMin=<-8 -3 -8>}
      localJointOffset=<0 -3 0>
     }]
    relativeSpaceBone {boneName=DYN_UpperCubeMSTR}
   }
   n53 = Get Cube_Float_A()
   n54 = Get Floating_Rotator2()
   n55 = "Transform (Modify) Bone - Bone: DYN_UpperCube1"(ComponentPose=@n52, Alpha=1, Translation=@n53, Rotation=@n54, Scale="1,1,1") {rotationMode=BMM_Additive, translationMode=BMM_Additive, alphaBoolBlend {blendOption=Linear}, boneToModify {boneName=DYN_UpperCube1}}
   n56 = "Transform (Modify) Bone - Bone: DYN_UpperCube2"(ComponentPose=@n55, Alpha=0.5, Translation=Get Cube_Float_B(), Rotation=@n54, Scale="1,1,1") {rotationMode=BMM_Additive, translationMode=BMM_Additive, alphaBoolBlend {blendOption=Linear}, boneToModify {boneName=DYN_UpperCube2}}
   n57 = Get Floating_Rotator1()
   n58 = "Transform (Modify) Bone - Bone: DYN_UpperCube3"(ComponentPose=@n56, Alpha=0.5, Translation=@n53, Rotation=@n57, Scale="1,1,1") {rotationMode=BMM_Additive, translationMode=BMM_Additive, alphaBoolBlend {blendOption=Linear}, boneToModify {boneName=DYN_UpperCube3}}
   n59 = Get Cube_Float_C()
   n60 = "Transform (Modify) Bone - Bone: DYN_UpperCube4"(ComponentPose=@n58, Alpha=1, Translation=@n59, Rotation=@n57, Scale="1,1,1") {rotationMode=BMM_Additive, translationMode=BMM_Additive, alphaBoolBlend {blendOption=Linear}, boneToModify {boneName=DYN_UpperCube4}}
   n61 = Get Cube_Float_D()
   n62 = "Transform (Modify) Bone - Bone: DYN_UpperCube5"(ComponentPose=@n60, Alpha=1, Translation=@n61, Rotation=@n57, Scale="1,1,1") {rotationMode=BMM_Additive, translationMode=BMM_Additive, alphaBoolBlend {blendOption=Linear}, boneToModify {boneName=DYN_UpperCube5}}
   n63 = "Transform (Modify) Bone - Bone: DYN_UpperCube6"(ComponentPose=@n62, Alpha=1, Translation=@n59, Rotation=@n54, Scale="1,1,1") {rotationMode=BMM_Additive, translationMode=BMM_Additive, alphaBoolBlend {blendOption=Linear}, boneToModify {boneName=DYN_UpperCube6}}
   n64 = "Transform (Modify) Bone - Bone: DYN_LowerCube2"(ComponentPose=@n63, Alpha=1, Translation=@n61, Rotation=@n57, Scale="1,1,1") {rotationMode=BMM_Additive, translationMode=BMM_Additive, alphaBoolBlend {blendOption=Linear}, boneToModify {boneName=DYN_LowerCube2}}
   Output Pose(Result=Blend Poses by bool(
     BlendPose_0=Component To Local(ComponentPose=@n64)
     BlendPose_1=Blend Poses by bool(BlendPose_0=@n44, BlendPose_1=@n22, BlendTime_0=1, BlendTime_1=1, bActiveValue=@n45) {blendType=Linear}
     bActiveValue=Get bIsInFrontEnd()
    ) {blendType=Linear}) {layerGroup=None}
   AnimGraphNode_SubInput_0 = Input Pose()
   n65 = "Transform (Modify) Bone - Bone: DYN_UpperCube4"(ComponentPose=Local To Component(LocalPose=@AnimGraphNode_SubInput_0), Alpha=1, Translation="-2,-1,-1", Rotation="0,0,0", Scale="1,1,1") {
    translationMode=BMM_Additive
    translationSpace=BCS_BoneSpace
    alphaBoolBlend {blendOption=Linear}
    boneToModify {boneName=DYN_UpperCube4}
    translation=<-2 -1 -1>
   }
   n66 = "Transform (Modify) Bone - Bone: DYN_UpperCube3"(ComponentPose=@n65, Alpha=1, Translation="-3,-3,1", Rotation="0,0,0", Scale="1,1,1") {
    translationMode=BMM_Additive
    translationSpace=BCS_BoneSpace
    alphaBoolBlend {blendOption=Linear}
    boneToModify {boneName=DYN_UpperCube3}
    translation=<-3 -3 1>
   }
   n67 = Get Cube_Float_A()
   n68 = Get Floating_Rotator2()
   n69 = "Transform (Modify) Bone - Bone: DYN_UpperCube1"(ComponentPose=@n66, Alpha=1, Translation=@n67, Rotation=@n68, Scale="1,1,1") {rotationMode=BMM_Additive, translationMode=BMM_Additive, alphaBoolBlend {blendOption=Linear}, boneToModify {boneName=DYN_UpperCube1}}
   n70 = "Transform (Modify) Bone - Bone: DYN_UpperCube2"(ComponentPose=@n69, Alpha=0.5, Translation=Get Cube_Float_B(), Rotation=@n68, Scale="1,1,1") {rotationMode=BMM_Additive, translationMode=BMM_Additive, alphaBoolBlend {blendOption=Linear}, boneToModify {boneName=DYN_UpperCube2}}
   n71 = Get Floating_Rotator1()
   n72 = "Transform (Modify) Bone - Bone: DYN_UpperCube3"(ComponentPose=@n70, Alpha=0.5, Translation=@n67, Rotation=@n71, Scale="1,1,1") {rotationMode=BMM_Additive, translationMode=BMM_Additive, alphaBoolBlend {blendOption=Linear}, boneToModify {boneName=DYN_UpperCube3}}
   n73 = Get Cube_Float_C()
   n74 = "Transform (Modify) Bone - Bone: DYN_UpperCube4"(ComponentPose=@n72, Alpha=1, Translation=@n73, Rotation=@n71, Scale="1,1,1") {rotationMode=BMM_Additive, translationMode=BMM_Additive, alphaBoolBlend {blendOption=Linear}, boneToModify {boneName=DYN_UpperCube4}}
   n75 = Get Cube_Float_D()
   n76 = "Transform (Modify) Bone - Bone: DYN_UpperCube5"(ComponentPose=@n74, Alpha=1, Translation=@n75, Rotation=@n71, Scale="1,1,1") {rotationMode=BMM_Additive, translationMode=BMM_Additive, alphaBoolBlend {blendOption=Linear}, boneToModify {boneName=DYN_UpperCube5}}
   n77 = "Transform (Modify) Bone - Bone: DYN_UpperCube6"(ComponentPose=@n76, Alpha=1, Translation=@n73, Rotation=@n68, Scale="1,1,1") {rotationMode=BMM_Additive, translationMode=BMM_Additive, alphaBoolBlend {blendOption=Linear}, boneToModify {boneName=DYN_UpperCube6}}
   n78 = "Transform (Modify) Bone - Bone: DYN_LowerCube2"(ComponentPose=@n77, Alpha=1, Translation=@n75, Rotation=@n71, Scale="1,1,1") {rotationMode=BMM_Additive, translationMode=BMM_Additive, alphaBoolBlend {blendOption=Linear}, boneToModify {boneName=DYN_LowerCube2}}
   n79 = Anim Dynamics - Bone: DYN_UpperCube1(ComponentPose=@n78, Alpha=0.9, ExternalForce="0,0,0") {
    angularDampingOverride=0.9
    angularSpringConstant=150
    bAngularSpring=true
    bGravityOverrideInSimSpace=true
    bLinearSpring=true
    bOverrideAngularDamping=true
    bOverrideLinearDamping=true
    gravityScale=0
    linearDampingOverride=0.995
    linearSpringConstant=25
    simulationSpace=BoneRelative
    alphaBoolBlend {blendOption=Linear}
    boundBone {boneName=DYN_UpperCube1}
    componentAppliedLinearAccClamp=<6000 6000 6000>
    componentLinearAccScale=<0.13 0.13 0.13>
    componentLinearVelScale=<0.9 0.9 0.9>
    physicsBodyDefinitions [{
      sphereCollisionRadius=0
      boundBone {boneName=DYN_UpperCube1}
      boxExtents=<50 50 50>
      constraintSetup {angularLimitsMax=<5 5 5>, angularLimitsMin=<-5 -5 -5>, angularTarget=<0 0 0>, linearAxesMax=<7 8 8>, linearAxesMin=<-8 -2 -5>}
      localJointOffset=<0 -3 0>
     }]
    relativeSpaceBone {boneName=Melee_TwoHanded_Handle}
   }
   n80 = Anim Dynamics - Bone: DYN_UpperCube2(ComponentPose=@n79, Alpha=0.9, ExternalForce="0,0,0") {
    angularDampingOverride=0.9
    angularSpringConstant=150
    bAngularSpring=true
    bGravityOverrideInSimSpace=true
    bLinearSpring=true
    bOverrideAngularDamping=true
    bOverrideLinearDamping=true
    gravityScale=0
    linearDampingOverride=0.995
    linearSpringConstant=25
    simulationSpace=BoneRelative
    alphaBoolBlend {blendOption=Linear}
    boundBone {boneName=DYN_UpperCube2}
    componentAppliedLinearAccClamp=<6000 6000 6000>
    componentLinearAccScale=<0.1 0.1 0.1>
    componentLinearVelScale=<0.9 0.9 0.9>
    physicsBodyDefinitions [{
      sphereCollisionRadius=0
      boundBone {boneName=DYN_UpperCube2}
      boxExtents=<50 50 50>
      constraintSetup {angularLimitsMax=<5 5 5>, angularLimitsMin=<-5 -5 -5>, angularTarget=<0 0 0>, linearAxesMax=<8 3 8>, linearAxesMin=<-5 -8 -4>}
      localJointOffset=<0 -3 0>
     }]
    relativeSpaceBone {boneName=Melee_TwoHanded_Handle}
   }
   n81 = Anim Dynamics - Bone: DYN_UpperCube3(ComponentPose=@n80, Alpha=0.7, ExternalForce="0,0,0") {
    angularDampingOverride=0.9
    angularSpringConstant=150
    bAngularSpring=true
    bGravityOverrideInSimSpace=true
    bLinearSpring=true
    bOverrideAngularDamping=true
    bOverrideLinearDamping=true
    gravityScale=0
    linearDampingOverride=0.995
    linearSpringConstant=25
    simulationSpace=BoneRelative
    alphaBoolBlend {blendOption=Linear}
    boundBone {boneName=DYN_UpperCube3}
    componentAppliedLinearAccClamp=<6000 6000 6000>
    componentLinearAccScale=<0.1 0.1 0.1>
    componentLinearVelScale=<0.9 0.9 0.9>
    physicsBodyDefinitions [{
      sphereCollisionRadius=0
      boundBone {boneName=DYN_UpperCube3}
      boxExtents=<50 50 50>
      constraintSetup {angularLimitsMax=<5 5 5>, angularLimitsMin=<-5 -5 -5>, angularTarget=<0 0 0>, linearAxesMax=<8 8 6>, linearAxesMin=<-7 0 0>}
      localJointOffset=<0 -3 0>
     }]
    relativeSpaceBone {boneName=Melee_TwoHanded_Handle}
   }
   n82 = Anim Dynamics - Bone: DYN_UpperCube4(ComponentPose=@n81, Alpha=0.9, ExternalForce="0,0,0") {
    angularDampingOverride=0.9
    angularSpringConstant=150
    bAngularSpring=true
    bGravityOverrideInSimSpace=true
    bLinearSpring=true
    bOverrideAngularDamping=true
    bOverrideLinearDamping=true
    gravityScale=0
    linearDampingOverride=0.995
    linearSpringConstant=25
    simulationSpace=BoneRelative
    alphaBoolBlend {blendOption=Linear}
    boundBone {boneName=DYN_UpperCube4}
    componentAppliedLinearAccClamp=<6000 6000 6000>
    componentLinearAccScale=<0.1 0.1 0.1>
    componentLinearVelScale=<0.9 0.9 0.9>
    physicsBodyDefinitions [{
      sphereCollisionRadius=0
      boundBone {boneName=DYN_UpperCube4}
      boxExtents=<50 50 50>
      constraintSetup {angularLimitsMax=<5 5 5>, angularLimitsMin=<-5 -5 -5>, angularTarget=<0 0 0>, linearAxesMax=<2 3 8>, linearAxesMin=<-8 -8 -2>}
      localJointOffset=<-3 0 0>
     }]
    relativeSpaceBone {boneName=Melee_TwoHanded_Handle}
   }
   n83 = Anim Dynamics - Bone: DYN_UpperCube5(ComponentPose=@n82, Alpha=0.9, ExternalForce="0,0,0") {
    angularDampingOverride=0.9
    angularSpringConstant=150
    bAngularSpring=true
    bGravityOverrideInSimSpace=true
    bLinearSpring=true
    bOverrideAngularDamping=true
    bOverrideLinearDamping=true
    gravityScale=0
    linearDampingOverride=0.995
    linearSpringConstant=25
    simulationSpace=BoneRelative
    alphaBoolBlend {blendOption=Linear}
    boundBone {boneName=DYN_UpperCube5}
    componentAppliedLinearAccClamp=<6000 6000 6000>
    componentLinearAccScale=<0.1 0.1 0.1>
    componentLinearVelScale=<0.9 0.9 0.9>
    physicsBodyDefinitions [{
      sphereCollisionRadius=0
      boundBone {boneName=DYN_UpperCube5}
      boxExtents=<50 50 50>
      constraintSetup {angularLimitsMax=<5 5 5>, angularLimitsMin=<-5 -5 -5>, angularTarget=<0 0 0>, linearAxesMax=<4 8 8>, linearAxesMin=<-8 -2 -8>}
      localJointOffset=<0 -3 0>
     }]
    relativeSpaceBone {boneName=Melee_TwoHanded_Handle}
   }
   n84 = Anim Dynamics - Bone: DYN_UpperCube6(ComponentPose=@n83, Alpha=0.7, ExternalForce="0,0,0") {
    angularDampingOverride=0.9
    angularSpringConstant=150
    bAngularSpring=true
    bGravityOverrideInSimSpace=true
    bLinearSpring=true
    bOverrideAngularDamping=true
    bOverrideLinearDamping=true
    gravityScale=0
    linearDampingOverride=0.995
    linearSpringConstant=50
    simulationSpace=BoneRelative
    alphaBoolBlend {blendOption=Linear}
    boundBone {boneName=DYN_UpperCube6}
    componentAppliedLinearAccClamp=<6000 6000 6000>
    componentLinearAccScale=<0.1 0.1 0.1>
    componentLinearVelScale=<0.9 0.9 0.9>
    physicsBodyDefinitions [{
      sphereCollisionRadius=0
      boundBone {boneName=DYN_UpperCube6}
      boxExtents=<50 50 50>
      constraintSetup {angularLimitsMax=<5 5 5>, angularLimitsMin=<-5 -5 -5>, angularTarget=<0 0 0>, linearAxesMax=<2 8 2>, linearAxesMin=<-8 -8 -8>}
      localJointOffset=<-3 0 0>
     }]
    relativeSpaceBone {boneName=Melee_TwoHanded_Handle}
   }
   n85 = Anim Dynamics - Bone: DYN_LowerCube2(ComponentPose=@n84, Alpha=0.9, ExternalForce="0,0,0") {
    angularDampingOverride=0.9
    angularSpringConstant=150
    bAngularSpring=true
    bGravityOverrideInSimSpace=true
    bLinearSpring=true
    bOverrideAngularDamping=true
    bOverrideLinearDamping=true
    gravityScale=0
    linearDampingOverride=0.995
    linearSpringConstant=25
    simulationSpace=BoneRelative
    alphaBoolBlend {blendOption=Linear}
    boundBone {boneName=DYN_LowerCube2}
    componentAppliedLinearAccClamp=<6000 6000 6000>
    componentLinearAccScale=<0.1 0.1 0.1>
    componentLinearVelScale=<0.9 0.9 0.9>
    physicsBodyDefinitions [{
      sphereCollisionRadius=0
      boundBone {boneName=DYN_LowerCube2}
      boxExtents=<50 50 50>
      constraintSetup {angularLimitsMax=<0 0 0>, angularLimitsMin=<0 0 0>, angularTarget=<0 0 0>, linearAxesMax=<2 8 8>, linearAxesMin=<-8 -3 -8>}
      localJointOffset=<0 -3 0>
     }]
    relativeSpaceBone {boneName=Melee_TwoHanded_Handle}
   }
   n86 = RigidBody(ComponentPose=@n85, Alpha=0.75) {
    bOverrideWorldGravity=true
    overridePhysicsAsset=&/Kevin_Couture_Male_Physics
    alphaBoolBlend {blendOption=Linear}
    baseBoneRef {boneName=b_AR_Root}
    overrideWorldGravity {z=-980}
   }
   Save cached pose 'OnGroundCache'(Pose=Component To Local(ComponentPose=@n86)) {cacheName=OnGroundCache}
  ]}
 EventGraph {nodes [
   Event BlueprintUpdateAnimation()
   > Sequence!(execute) -> (then_0, then_1, then_2:Exec)
   n1 = GetGameTimeInSeconds()
   n2 = "Sin (Radians)"(A=float * float(A=float * float(A=1, B=1), B=@n1))
   n3 = "Sin (Radians)"(A=float * float(A=float * float(A=1, B=2), B=@n1))
   n4 = "Sin (Radians)"(A=float * float(A=float * float(A=1, B=1.5), B=@n1))
   > Set Cube_Float_A(Cube_Float_A:Vector=MakeVector(X=float * float(A=@n4, B=0.5), Y=float * float(A=@n2, B=0.76), Z=float * float(A=@n3, B=0.42))) -> (Output_Get:Vector)
   n5 = "Cos (Radians)"!(A:double=float * float(A=float * float(A=1, B=1.5), B=@n1)) -> (ReturnValue:double)
   n6 = "Cos (Radians)"(A=float * float(A=float * float(A=1, B=1.78), B=@n1))
   n7 = "Cos (Radians)"(A=float * float(A=float * float(A=1, B=1.15), B=@n1))
   > Set Cube_Float_B(Cube_Float_B:Vector=MakeVector(X=float * float(A=@n5, B=0.8), Y=float * float(A=@n6, B=0.925), Z=float * float(A=@n7, B=0.5))) -> (Output_Get:Vector)
   n8 = float * float(A=float * float(A=1, B=1.5), B=@n1)
   n9 = float * float(A=float * float(A=1, B=1), B=@n1)
   n10 = float * float(A=float * float(A=1, B=2), B=@n1)
   n11 = MakeVector(
    X=float * float(A="Sin (Radians)"(A=@n8), B=0.5)
    Y=float * float(A="Sin (Radians)"(A=@n9), B=0.8)
    Z=float * float(A="Sin (Radians)"(A=@n10), B=0.45)
   )
   > Set Cube_Float_C(Cube_Float_C:Vector=@n11) -> (Output_Get:Vector)
   n12 = float * float(A=float * float(A=0.6678, B=1.5), B=@n1)
   n13 = float * float(A=float * float(A=1, B=1.47), B=@n1)
   n14 = float * float(A=float * float(A=1, B=0.953), B=@n1)
   n15 = MakeVector(
    X=float * float(A="Sin (Radians)"(A=@n12), B=1.05)
    Y=float * float(A="Sin (Radians)"(A=@n14), B=0.3)
    Z=float * float(A="Sin (Radians)"(A=@n13), B=0.9)
   )
   > Set Cube_Float_D(Cube_Float_D:Vector=@n15) -> (Output_Get:Vector)
   n16 = BreakVector(InVec=vector * float(A=@n15, B=8))
   > Set Floating_Rotator2(Floating_Rotator2:Rotator=MakeRotator(Roll=@n16, Pitch=@n16.Y, Yaw=@n16.Z)) -> (Output_Get:Rotator)
   n17 = BreakVector(InVec=vector * float(A=@n11, B=11))
   > Set Floating_Rotator1(Floating_Rotator1:Rotator=MakeRotator(Roll=@n17, Pitch=@n17.Y, Yaw=@n17.Z)) -> (Output_Get:Rotator)
  ]}
]
variables [
 Cube_Float_A {type=Vector}
 Cube_Float_B {type=Vector}
 Cube_Float_C {type=Vector}
 Cube_Float_D {type=Vector}
 Floating_Rotator1 {type=Rotator}
 Floating_Rotator2 {type=Rotator}
]
